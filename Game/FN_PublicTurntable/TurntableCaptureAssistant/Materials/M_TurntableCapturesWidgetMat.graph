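# UI-domain material for the Turntable Capture Assistant, named for showing turntable captures in a widget.
# It holds a single texture sample parameter, Texture, which defaults to the TurntableCapture_RT0 render target.
# The parameter lets material instances substitute another texture or capture target.
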Material /Game/FN_PublicTurntable/TurntableCaptureAssistant/Materials/M_TurntableCapturesWidgetMat
graph {
 properties {materialDomain=MD_UI, pixelDepthOffsetMode=PDOM_Legacy, usedShadingModels=Default Lit}
 nodes [TextureSampleParameter2D() {parameterName=Texture, texture=/Game/FN_PublicTurntable/TurntableCaptureAssistant/RenderTargets/TurntableCapture_RT0}]
}
