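# Full-screen UMG widget that displays a turntable capture render target with a text label in the bottom-left.
# On Construct it creates a viewport-sized render target, stores it in currentRT and feeds it to the image's dynamic material through the Texture parameter.
# setNewTexture swaps that parameter to a supplied render target and sets the rich text label to the capture data's FriendlyName.
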WidgetBlueprint /Game/FN_PublicTurntable/TurntableCaptureAssistant/UI/TurntableCaptures_RenderDisplayWidget
graphs [
 EventGraph {nodes [
   CreateRenderTarget2D!(
    execute
    Width:int=256
    Height:int=256
    Format:ETextureRenderTargetFormat Enum=RTF_RGBA16f
    ClearColor:Linear Color Structure=(R=0,G=0,B=0,A=1)
    bAutoGenerateMipMaps:bool=false
    bSupportUAVs:bool=false
   ) -> (then, ReturnValue:Texture Render Target 2D Object Reference)
   > SetBrushFromTexture!(
    execute
    self:Image Object Reference=Get Image_45() -> (Image_45:Image Object Reference)
    Texture:Texture 2D Object Reference
    bMatchSize:bool=false
   ) -> (then)
   Event Construct()
   n1 = GetViewportSize!() -> (ReturnValue_X:double, ReturnValue_Y:double)
   > n2 = CreateRenderTarget2D(
    Width=Truncate(A=@n1)
    Height=Truncate(A=@n1.ReturnValue_Y)
    Format=RTF_RGBA16f
    ClearColor=(R=0,G=0,B=0,A=1)
    bAutoGenerateMipMaps=false
    bSupportUAVs=false
   )
   > Set currentRT(currentRT:Texture Render Target 2D Object Reference=@n2) -> (Output_Get:Texture Render Target 2D Object Reference)
   > n3 = GetDynamicMaterial(self=Get Image_45())
   > SetTextureParameterValue(self=@n3, ParameterName=Texture, Value=Get currentRT() -> (currentRT:Texture Render Target 2D Object Reference))
  ]}
 setNewTexture {nodes [
   n1 = setNewTexture!() -> (then, inMaterialInstance:Texture Render Target Object Reference, inCaptureData:Turntable Captures Data Struct Structure) {is=title K2Node_FunctionEntry}
   > n2 = GetDynamicMaterial(self=Get Image_45())
   > SetTextureParameterValue(self=@n2, ParameterName=Texture, Value=@n1)
   > SetText(
    self=Get RichTextBlock_121() -> (RichTextBlock_121:Rich Text Block Object Reference)
    InText="To Text (String)"(InString=Break Turntable Captures Data Struct!(TurntableCapturesData_Struct:"Turntable Captures Data Struct Structure (by ref)"=@n1.inCaptureData) -> (FriendlyName_5_#1:String, Material_6_#2:Material Interface Object Reference) {is=title K2Node_BreakStruct})
   )
  ]}
]
variables [currentRT {type=TextureRenderTarget2D, default=null}]
widgets {
 info {inheritedWidgetCount=0, parentClass=&UserWidget, rootWidgetClass=&CanvasPanel, widgetCount=3}
 widgets [
  CanvasPanel_0 {widgetClassPath=&CanvasPanel, widget=CanvasPanel_0}
  Image_45 {
   widgetClassPath=&Image
   bIsVariable=true
   parent=CanvasPanel_0
   widget=Image_45
   properties {brush {
     bIsDynamicallyLoaded=false
     drawAs=Image
     imageType=NoImage
     mirroring=NoMirror
     resourceName=None
     resourceObject=/Game/FN_PublicTurntable/TurntableCaptureAssistant/Materials/M_TurntableCapturesWidgetMat
     tiling=NoTile
     imageSize=<32 32>
     margin=<ltrb 0 0 0 0>
     outlineSettings {
      bUseBrushTransparency=false
      roundingType=HalfHeightRadius
      width=0
      color {colorUseRule=UseColor_Specified, specifiedColor=<rgba 0 0 0 0>}
      cornerRadii=<0 0 0 0>
     }
     tintColor {colorUseRule=UseColor_Specified, specifiedColor=<rgba 1 1 1 1>}
     uVRegion {bIsValid=false, max=<0 0>, min=<0 0>}
    }}
   slot {class=CanvasPanelSlot, properties {layoutData {alignment=<0 0>, anchors {maximum=<1 1>, minimum=<0 0>}, offsets=<ltrb 0 0 0 0>}}}
  }
  RichTextBlock_121 {
   widgetClassPath=/Script/UMG.RichTextBlock
   bIsVariable=true
   parent=CanvasPanel_0
   widget=RichTextBlock_121
   properties {textStyleSet=/Game/FN_PublicTurntable/TurntableCaptureAssistant/Data/TurntableCaptures_TextStyles_Data}
   slot {
    class=CanvasPanelSlot
    properties {bAutoSize=true, layoutData {alignment=<0 1>, anchors {maximum=<1 1>, minimum=<0 1>}, offsets=<ltrb 5 -5 0 30>}}
   }
  }
 ]
}
